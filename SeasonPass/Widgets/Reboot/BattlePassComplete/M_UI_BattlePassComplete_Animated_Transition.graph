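# Translucent UI material for the animated Battle Pass completion transition, drawing the badge icon with completion colors, glows and highlights.
# Color dodge and lerp blends recolor the badge, with UsePureTint fading to a flat tint.
# Named reroutes supply a glow SDF texture, panning noise, outline layers and a time-pulsed under glow.
# Animate-group scalars drive sine-based outline and rotated sheen swipes, so the transition can be animated externally.
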Material /SeasonPass/Widgets/Reboot/BattlePassComplete/M_UI_BattlePassComplete_Animated_Transition
graph {
 properties {
  bUsedWithStaticMesh=false
  blendMode=BLEND_Translucent
  materialDomain=MD_UI
  pixelDepthOffsetMode=PDOM_Legacy
  usedShadingModels=Default Lit
 }
 nodes [
  n1 = NamedRerouteUsage()
  n2 = TextureSample(UVs=MaterialFunctionCall(Texture Scale=Constant() {r=0.8}) {materialFunction=&ScaleUVsByCenter}.UVs) {texture=/Game/UI/Foundation/Textures/Icons/Items/T-FNBR-BattlePass-L}
  n3 = NamedRerouteDeclaration(None=@n2.RGB) {name=Badge Color, nodeColor=<rgba 0 0 0 1>}
  n4 = LinearInterpolate(
   A=NamedRerouteUsage()
   B=Saturate(None=MaterialFunctionCall!(Blend=@n1, Base=@n3) {materialFunction=/Engine/Functions/Engine_MaterialFunctions03/Blends/Blend_ColorDodge}.Result)
   Alpha=NamedRerouteUsage()
  )
  n5 = OneMinus(None=Desaturation(None=@n4))
  n6 = NamedRerouteUsage()
  n7 = NamedRerouteUsage()
  n8 = LinearInterpolate(
   A=LinearInterpolate(A=@n4, B=NamedRerouteUsage(), Alpha=Multiply(A=@n5, B=@n6))
   B=LinearInterpolate(A=NamedRerouteUsage(), B=NamedRerouteUsage(), Alpha=Multiply(A=@n7, B=@n7))
   Alpha=Power(Base=@n7) {constExponent=0.5}
  )
  n9 = Power(Base=NamedRerouteUsage()) {constExponent=0.2}
  n10 = NamedRerouteUsage()
  LinearInterpolate(
   A=MaterialFunctionCall(Blend=NamedRerouteUsage(), Base=LinearInterpolate(A=@n8, B=@n9, Alpha=@n10) {constA=1}) {materialFunction=&Blend_ColorDodge}.Result
   B=VectorParameter() {group=Color, parameterName=PureTintColor, defaultValue=<rgba 1 0.8417 0.05 1>}.RGB
   Alpha=ScalarParameter() {group=Animate, parameterName=UsePureTint}
  )
  LinearInterpolate(A=NamedRerouteUsage(), Alpha=NamedRerouteUsage())
  MaterialFunctionCall() {materialFunction=&DebugTimeSine}
  NamedRerouteDeclaration(None=@n2.A) {name=Badge Alpha, nodeColor=<rgba 0 0 0 1>}
  n11 = SmoothStep(Value=NamedRerouteUsage()) {constMax=0.5}
  n12 = Multiply(A=@n11, B=@n11)
  n13 = NamedRerouteUsage()
  n14 = Multiply(A=@n13, B=@n13)
  n15 = NamedRerouteDeclaration(None=Multiply(A=Multiply(A=@n12, B=@n14), B=NamedRerouteUsage())) {name=Outline Highlight, nodeColor=<rgba 0 0 0 1>}
  n16 = NamedRerouteDeclaration(None=Multiply(A=SmoothStep(Value=@n13) {constMin=0.9}, B=NamedRerouteUsage())) {name=Outline, nodeColor=<rgba 0 0 0 1>}
  n17 = SmoothStep(Value=NamedRerouteUsage()) {constMin=0.5}
  n18 = NamedRerouteUsage()
  n19 = MaterialFunctionCall!("Value (0-1)"=NamedRerouteUsage()) {materialFunction=/FortUILibrary/MaterialFunctions/Basics/MF_UI_EaseCurves_01_Sine}
  n20 = NamedRerouteUsage()
  n21 = Multiply(A=Multiply(A=@n20, B=@n20), B=@n20)
  n22 = Multiply(A=LinearInterpolate(A=Multiply(A=@n17, B=@n18), Alpha=Multiply(A=@n19.In-Out, B=@n21)), B=NamedRerouteUsage())
  NamedRerouteDeclaration(None=LinearInterpolate(A=LinearInterpolate(A=@n15, Alpha=@n16), Alpha=NamedRerouteDeclaration(None=@n22) {name=Outline Glow, nodeColor=<rgba 0 0 0 1>})) {name=Outline Alpha, nodeColor=<rgba 0 0 0 1>}
  NamedRerouteDeclaration(None=Constant3Vector() {constant=<rgba 0 0.0331 0.0296 1>}) {name=Complete Color 2, nodeColor=<rgba 0 0 0 1>}
  NamedRerouteDeclaration(None=Constant3Vector() {constant=<rgba 0.0395 1 0.2307 1>}) {name=Complete Color 1, nodeColor=<rgba 0 0 0 1>}
  n23 = MaterialFunctionCall() {materialFunction=&GetUserInterfaceUV}
  n24 = Constant2Vector() {g=0.2}
  n25 = Multiply(A=Distance(A=Add(A=@n23.Normalized UV, B=@n24), B=Constant() {r=0.5})) {constB=1.5}
  n26 = Multiply(A=@n25, B=@n25) {constB=2}
  n27 = Multiply(A=@n26, B=@n26) {constB=2}
  n28 = Multiply(A=Time()) {constB=0.5}
  n29 = Divide(A=Add(A=Sine(None=@n28)))
  NamedRerouteDeclaration(None=Multiply(A=LinearInterpolate(A=@n26, B=@n27, Alpha=@n29), B=ScalarParameter() {defaultValue=0.35, parameterName=Under Glow Opacity})) {name=UnderGlow, nodeColor=<rgba 0 0 0 1>}
  n30 = ScalarParameter() {defaultValue=0.5, group=Animate, parameterName=OutlineSwipe}
  n31 = Multiply(A=Add(A=TextureCoordinate(), B=LinearInterpolate(Alpha=@n30) {constA=-1})) {constB=0.5}
  n32 = Saturate(None=ComponentMask(None=Sine(None=@n31)) {g=true})
  NamedRerouteDeclaration(None=Multiply(A=Multiply(A=@n32, B=@n32), B=ScalarParameter() {defaultValue=1, parameterName=Outline Opacity})) {name=Outline Mask, nodeColor=<rgba 0 0 0 1>}
  n33 = TextureCoordinate()
  n34 = ScalarParameter() {defaultValue=0.125, group=Animate, parameterName=SheenAngle}
  n35 = Multiply(A=Add(A=Rotator(Coordinate=@n33, Time=@n34) {speed=6.2832}, B=ScalarParameter() {defaultValue=1, group=Animate, parameterName=SheenSwipe})) {constB=0.5}
  n36 = Saturate(None=ComponentMask(None=Sine(None=@n35)) {g=true})
  n37 = Multiply(A=@n36, B=@n36)
  NamedRerouteDeclaration(None=Multiply(A=@n37, B=@n37)) {name=Sheen Swipe, nodeColor=<rgba 0 0 0 1>}
  n38 = Panner() {speedY=1}
  n39 = TextureObjectParameter() {parameterName=Noise, samplerType=SAMPLERTYPE_Grayscale, texture=&LowResBlurredNoise}
  n40 = TextureSample(
   UVs=MaterialFunctionCall(UVs=@n38, Texture Scale=Constant() {r=3}) {materialFunction=&ScaleUVsByCenter}.UVs
   Tex=@n39
  ) {samplerType=SAMPLERTYPE_Grayscale, texture=&LowResBlurredNoise}
  n41 = TextureSample(
   UVs=MaterialFunctionCall(UVs=@n38, Texture Scale=Constant() {r=2}) {materialFunction=&ScaleUVsByCenter}.UVs
   Tex=@n39
  ) {samplerType=SAMPLERTYPE_Grayscale, texture=&LowResBlurredNoise}
  NamedRerouteDeclaration(None=Saturate(None=Add(A=@n40.R, B=@n41.R))) {name=Noise, nodeColor=<rgba 0 0 0 1>}
  NamedRerouteDeclaration(None=TextureSample() {texture=&/T_UI_Icon_BattlePass_Glow}.R) {name=GlowSDF, nodeColor=<rgba 0 0 0 1>}
  n42 = Add(A=@n38) {constB=0.5}
  n43 = Constant() {r=3}
  NamedRerouteDeclaration(None=TextureSample(UVs=MaterialFunctionCall(UVs=@n42, Texture Scale=@n43) {materialFunction=&ScaleUVsByCenter}.UVs) {samplerType=SAMPLERTYPE_Grayscale, texture=&LowResBlurredNoise}.R) {name=Highlight Noise, nodeColor=<rgba 0 0 0 1>}
 ]
}
