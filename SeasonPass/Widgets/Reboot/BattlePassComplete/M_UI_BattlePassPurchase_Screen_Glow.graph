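# Additive UI material that draws a revealable green radial glow for the Battle Pass purchase completion screen.
# UV distance from center feeds a SmoothStep to form the Base Glow mask, alongside a Reveal_Glow parameter, both shared through named reroutes.
# Color lerps between two greens and passes through a Color Dodge blend driven by a powered reroute value.
# Intensity combines an EaseInOutSine curve, another reroute value and the GlowOpacity parameter.
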Material /SeasonPass/Widgets/Reboot/BattlePassComplete/M_UI_BattlePassPurchase_Screen_Glow
graph {
 properties {bAutomaticallySetUsageInEditor=false, blendMode=BLEND_Additive, materialDomain=MD_UI, usedShadingModels=Default Lit}
 nodes [
  n1 = Power(Base=NamedRerouteUsage())
  MaterialFunctionCall!(
   Blend=@n1
   Base=LinearInterpolate(A=Constant3Vector() {constant=<rgba 0 1 0.2469 1>}, B=Constant3Vector() {constant=<rgba 0.2848 0.6198 0 1>}, Alpha=@n1)
  ) {materialFunction=/Engine/Functions/Engine_MaterialFunctions03/Blends/Blend_ColorDodge}
  n2 = NamedRerouteUsage()
  Multiply(
   A=Multiply(
    A=MaterialFunctionCall("Value (0-1)"=@n2) {materialFunction=&MF_UI_EaseCurves}.EaseInOutSine
    B=NamedRerouteUsage()
   )
   B=ScalarParameter() {defaultValue=0.75, parameterName=GlowOpacity}
  )
  NamedRerouteDeclaration(None=ScalarParameter() {defaultValue=1, parameterName=Reveal_Glow}) {name=Reveal Glow, nodeColor=<rgba 0 0 0 1>}
  n3 = NamedRerouteUsage()
  n4 = TextureCoordinate()
  n5 = Constant() {r=0.5}
  n6 = OneMinus(None=Multiply(A=Distance(A=@n4, B=@n5)) {constB=2})
  NamedRerouteDeclaration(None=SmoothStep(Min=OneMinus(None=@n3), Value=Saturate(None=@n6))) {name=Base Glow, nodeColor=<rgba 0 0 0 1>}
 ]
}
